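# Physics asset for the dynamic chain bones on the male medium Tech Explorer player body, at the right thigh.
# Front and back master bones connect to a thigh-twist capsule with free angular constraints.
# Small capsules and boxes form a four-link front strand and two-link back strand, with limited swing and zero-degree twist limits.
# A fully angular-locked constraint joins both strand ends, closing the chain into a loop.
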PhysicsAsset /BRCosmetics/Characters/Player/Male/Medium/Bodies/M_MED_Tech_Explorer/Meshes/M_MED_Tech_Explorer_chains_dyn
bodies {
 dyn_chains_01_bk [{shapeType=Capsule, length=5, radius=1, center=<5 0 0>, rotation=<rot 90 0 0>}]
 dyn_chains_01_ft [{shapeType=Box, extentX=2, extentY=2, extentZ=1, center=<1 0 0>}]
 dyn_chains_02_bk [{shapeType=Capsule, length=5, radius=1, center=<4 0 0>, rotation=<rot 90 0 0>}]
 dyn_chains_02_ft [{shapeType=Box, extentX=7.5, extentY=6.8135, extentZ=1, center=<4.2132 0 -0.0001>}]
 dyn_chains_03_ft [{shapeType=Capsule, length=5, radius=1, center=<5 -0.5991 0>, rotation=<rot 90 0 0>}]
 dyn_chains_04_ft [{shapeType=Capsule, length=5, radius=1, center=<4 -0.2558 0>, rotation=<rot 90 0 0>}]
 dyn_chains_bk_MSTR [{shapeType=Capsule, length=1, radius=1}]
 dyn_chains_ft_MSTR [{shapeType=Capsule, length=1, radius=1}]
 thigh_twist_01_r [{shapeType=Capsule, length=30.2225, radius=8.5481, center=<-8.0555 0.0116 -2.3509>, rotation=<rot 87.2652 2.0165 -0.0001>}]
}
constraints [
 {
  bone1Name=dyn_chains_ft_MSTR
  bone2Name=thigh_twist_01_r
  swing1LimitDegrees=45
  swing1Motion=Free
  swing2LimitDegrees=45
  swing2Motion=Free
  twistLimitDegrees=45
  twistMotion=Free
 }
 {bone1Name=dyn_chains_01_ft, bone2Name=dyn_chains_ft_MSTR, swing1LimitDegrees=40, swing2LimitDegrees=40, twistLimitDegrees=0}
 {bone1Name=dyn_chains_02_ft, bone2Name=dyn_chains_01_ft, swing1LimitDegrees=45, swing2LimitDegrees=45, twistLimitDegrees=0}
 {bone1Name=dyn_chains_03_ft, bone2Name=dyn_chains_02_ft, swing1LimitDegrees=45, swing2LimitDegrees=45, twistLimitDegrees=0}
 {bone1Name=dyn_chains_04_ft, bone2Name=dyn_chains_03_ft, swing1LimitDegrees=45, swing2LimitDegrees=45, twistLimitDegrees=0}
 {
  bone1Name=dyn_chains_bk_MSTR
  bone2Name=thigh_twist_01_r
  swing1LimitDegrees=45
  swing1Motion=Free
  swing2LimitDegrees=45
  swing2Motion=Free
  twistLimitDegrees=45
  twistMotion=Free
 }
 {bone1Name=dyn_chains_01_bk, bone2Name=dyn_chains_bk_MSTR, swing1LimitDegrees=40, swing2LimitDegrees=40, twistLimitDegrees=0}
 {bone1Name=dyn_chains_02_bk, bone2Name=dyn_chains_01_bk, swing1LimitDegrees=45, swing2LimitDegrees=45, twistLimitDegrees=0}
 {
  bone1Name=dyn_chains_04_ft
  bone2Name=dyn_chains_02_bk
  swing1LimitDegrees=45
  swing1Motion=Locked
  swing2LimitDegrees=45
  swing2Motion=Locked
  twistLimitDegrees=5
  twistMotion=Locked
 }
]
