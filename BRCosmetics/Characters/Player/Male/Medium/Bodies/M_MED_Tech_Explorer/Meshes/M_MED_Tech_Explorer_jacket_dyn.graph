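# Physics asset for the Tech Explorer male medium body's jacket, simulating the dynamic jacket skirt around hips and thighs.
# Box-shaped segment chains for front, side and back panels hang with limited swing and twist from small master capsules jointed to the pelvis.
# Cross constraints tie neighboring lower segments into a hem, open at the front.
# Capsules on pelvis, lower spine and thighs supply body shapes, with extra ones on pelvis and right thigh.
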PhysicsAsset /BRCosmetics/Characters/Player/Male/Medium/Bodies/M_MED_Tech_Explorer/Meshes/M_MED_Tech_Explorer_jacket_dyn
bodies {
 dyn_jacket_bk [{shapeType=Box, extentX=10, extentY=14, extentZ=2.2, center=<5 0 0>}]
 dyn_jacket_bk_MSTR [{shapeType=Capsule, length=1, radius=1}]
 dyn_jacket_bk_l_01 [{shapeType=Box, extentX=11, extentY=7, extentZ=2, center=<5.5 3.5 0>}]
 dyn_jacket_bk_r_01 [{shapeType=Box, extentX=11, extentY=7, extentZ=2, center=<5.5 -3.5 0>}]
 dyn_jacket_l_01 [
  {shapeType=Box, extentX=10, extentY=6, extentZ=2, center=<5 -3 0>, rotation=<rot 0 0 -5>}
  {shapeType=Box, extentX=10, extentY=6, extentZ=2, center=<5 3 -0.8>, rotation=<rot 0 0 20>}
 ]
 dyn_jacket_l_02 [
  {shapeType=Box, extentX=10, extentY=6, extentZ=2, center=<5 -3 0>, rotation=<rot 0 0 -5>}
  {shapeType=Box, extentX=10, extentY=6, extentZ=2, center=<5 3 -0.8>, rotation=<rot 0 0 20>}
 ]
 dyn_jacket_l_MSTR [{shapeType=Capsule, length=1, radius=1}]
 dyn_jacket_l_bk_01 [{shapeType=Box, extentX=10, extentY=12, extentZ=2, center=<5 0 0>}]
 dyn_jacket_l_bk_02 [{shapeType=Box, extentX=10, extentY=12, extentZ=2, center=<5 0 0>}]
 dyn_jacket_l_bk_MSTR [{shapeType=Capsule, length=1, radius=1}]
 dyn_jacket_l_fr_01 [
  {shapeType=Box, extentX=11, extentY=6, extentZ=2, center=<6 0 -0.5>}
  {shapeType=Box, extentX=11, extentY=6, extentZ=2, center=<6 -4 -1>, rotation=<rot 0 0 -20>}
 ]
 dyn_jacket_l_fr_02 [
  {shapeType=Box, extentX=11, extentY=6, extentZ=2, center=<6 0 -0.5>}
  {shapeType=Box, extentX=11, extentY=6, extentZ=2, center=<6 -4 -1>, rotation=<rot 0 0 -20>}
 ]
 dyn_jacket_l_fr_MSTR [{shapeType=Capsule, length=1, radius=1}]
 dyn_jacket_r_01 [
  {shapeType=Box, extentX=10, extentY=6, extentZ=2, center=<5 3 0>, rotation=<rot 0 0 5>}
  {shapeType=Box, extentX=10, extentY=6, extentZ=2, center=<5 -3 -0.8>, rotation=<rot 0 0 -20>}
 ]
 dyn_jacket_r_02 [
  {shapeType=Box, extentX=10, extentY=6, extentZ=2, center=<5 3 0>, rotation=<rot 0 0 5>}
  {shapeType=Box, extentX=10, extentY=6, extentZ=2, center=<5 -3 -0.8>, rotation=<rot 0 0 -20>}
 ]
 dyn_jacket_r_MSTR [{shapeType=Capsule, length=1, radius=1}]
 dyn_jacket_r_bk_01 [{shapeType=Box, extentX=10, extentY=12, extentZ=2, center=<5 0 0>}]
 dyn_jacket_r_bk_02 [{shapeType=Box, extentX=10, extentY=12, extentZ=2, center=<5 0 0>}]
 dyn_jacket_r_bk_MSTR [{shapeType=Capsule, length=1, radius=1}]
 dyn_jacket_r_fr_01 [
  {shapeType=Box, extentX=11, extentY=6, extentZ=2, center=<6 0 -0.5>}
  {shapeType=Box, extentX=11, extentY=6, extentZ=2, center=<6 4 -1>, rotation=<rot 0 0 20>}
 ]
 dyn_jacket_r_fr_02 [
  {shapeType=Box, extentX=11, extentY=6, extentZ=2, center=<6 0 -0.5>}
  {shapeType=Box, extentX=11, extentY=6, extentZ=2, center=<6 4 -1>, rotation=<rot 0 0 20>}
 ]
 dyn_jacket_r_fr_MSTR [{shapeType=Capsule, length=1, radius=1}]
 pelvis [
  {shapeType=Capsule, length=8.8566, radius=11.1109, center=<-4.1085 -0.1339 0>}
  {shapeType=Capsule, length=10.596, radius=3.8151, center=<-5.4259 -6.8297 9.4194>, rotation=<rot 81.9977 0.1459 -1.1206>}
 ]
 spine_01 [{shapeType=Capsule, length=7.8452, radius=10.044, center=<-1.9461 0.312 0>}]
 thigh_twist_01_l [{shapeType=Capsule, length=26.4162, radius=9.043, center=<8.019 -0.9026 1.2202>, rotation=<rot 88.1882 5.0001 5.3096>}]
 thigh_twist_01_r [
  {shapeType=Capsule, length=26.4162, radius=9.043, center=<-8.019 0.9026 -1.2202>, rotation=<rot 88.1882 5.0001 5.3096>}
  {shapeType=Capsule, length=12.2038, radius=4.6103, center=<-15.9738 -5.3467 -0.4622>, rotation=<rot 90 0 0>}
  {shapeType=Capsule, length=13.029, radius=3.7851, center=<-18.1173 6.3927 1.2414>, rotation=<rot 90 0 0>}
 ]
}
constraints [
 {bone1Name=spine_01, bone2Name=pelvis, swing1LimitDegrees=45, swing2LimitDegrees=45, twistLimitDegrees=45}
 {bone1Name=dyn_jacket_bk, bone2Name=dyn_jacket_bk_MSTR, swing1LimitDegrees=30, swing2LimitDegrees=30, twistLimitDegrees=20}
 {bone1Name=dyn_jacket_bk_l_01, bone2Name=dyn_jacket_bk, swing1LimitDegrees=30, swing2LimitDegrees=30, twistLimitDegrees=20}
 {bone1Name=dyn_jacket_bk_r_01, bone2Name=dyn_jacket_bk, swing1LimitDegrees=30, swing2LimitDegrees=30, twistLimitDegrees=20}
 {bone1Name=dyn_jacket_r_bk_01, bone2Name=dyn_jacket_r_bk_MSTR, swing1LimitDegrees=30, swing2LimitDegrees=30, twistLimitDegrees=20}
 {bone1Name=dyn_jacket_r_bk_02, bone2Name=dyn_jacket_r_bk_01, swing1LimitDegrees=30, swing2LimitDegrees=30, twistLimitDegrees=20}
 {bone1Name=dyn_jacket_l_bk_01, bone2Name=dyn_jacket_l_bk_MSTR, swing1LimitDegrees=30, swing2LimitDegrees=30, twistLimitDegrees=20}
 {bone1Name=dyn_jacket_l_bk_02, bone2Name=dyn_jacket_l_bk_01, swing1LimitDegrees=30, swing2LimitDegrees=30, twistLimitDegrees=20}
 {bone1Name=dyn_jacket_l_fr_01, bone2Name=dyn_jacket_l_fr_MSTR, swing1LimitDegrees=45, swing2LimitDegrees=30, twistLimitDegrees=10}
 {bone1Name=dyn_jacket_l_fr_02, bone2Name=dyn_jacket_l_fr_01, swing1LimitDegrees=35, swing2LimitDegrees=75, twistLimitDegrees=20}
 {bone1Name=dyn_jacket_r_fr_01, bone2Name=dyn_jacket_r_fr_MSTR, swing1LimitDegrees=45, swing2LimitDegrees=30, twistLimitDegrees=20}
 {bone1Name=dyn_jacket_r_fr_02, bone2Name=dyn_jacket_r_fr_01, swing1LimitDegrees=35, swing2LimitDegrees=75, twistLimitDegrees=20}
 {bone1Name=dyn_jacket_l_02, bone2Name=dyn_jacket_l_01, swing1LimitDegrees=30, swing2LimitDegrees=30, twistLimitDegrees=20}
 {bone1Name=dyn_jacket_l_01, bone2Name=dyn_jacket_l_MSTR, swing1LimitDegrees=30, swing2LimitDegrees=30, twistLimitDegrees=20}
 {bone1Name=dyn_jacket_r_01, bone2Name=dyn_jacket_r_MSTR, swing1LimitDegrees=30, swing2LimitDegrees=30, twistLimitDegrees=20}
 {bone1Name=dyn_jacket_r_02, bone2Name=dyn_jacket_r_01, swing1LimitDegrees=30, swing2LimitDegrees=30, twistLimitDegrees=20}
 {
  bone1Name=dyn_jacket_r_bk_MSTR
  bone2Name=pelvis
  swing1LimitDegrees=45
  swing1Motion=Free
  swing2LimitDegrees=45
  swing2Motion=Free
  twistLimitDegrees=45
  twistMotion=Free
 }
 {
  bone1Name=dyn_jacket_l_bk_MSTR
  bone2Name=pelvis
  swing1LimitDegrees=45
  swing1Motion=Free
  swing2LimitDegrees=45
  swing2Motion=Free
  twistLimitDegrees=45
  twistMotion=Free
 }
 {
  bone1Name=dyn_jacket_bk_MSTR
  bone2Name=pelvis
  swing1LimitDegrees=45
  swing1Motion=Free
  swing2LimitDegrees=45
  swing2Motion=Free
  twistLimitDegrees=45
  twistMotion=Free
 }
 {
  bone1Name=dyn_jacket_r_MSTR
  bone2Name=pelvis
  swing1LimitDegrees=45
  swing1Motion=Free
  swing2LimitDegrees=45
  swing2Motion=Free
  twistLimitDegrees=45
  twistMotion=Free
 }
 {
  bone1Name=dyn_jacket_l_MSTR
  bone2Name=pelvis
  swing1LimitDegrees=45
  swing1Motion=Free
  swing2LimitDegrees=45
  swing2Motion=Free
  twistLimitDegrees=45
  twistMotion=Free
 }
 {
  bone1Name=dyn_jacket_l_fr_MSTR
  bone2Name=pelvis
  swing1LimitDegrees=45
  swing1Motion=Free
  swing2LimitDegrees=45
  swing2Motion=Free
  twistLimitDegrees=45
  twistMotion=Free
 }
 {
  bone1Name=dyn_jacket_r_fr_MSTR
  bone2Name=pelvis
  swing1LimitDegrees=45
  swing1Motion=Free
  swing2LimitDegrees=45
  swing2Motion=Free
  twistLimitDegrees=45
  twistMotion=Free
 }
 {
  bone1Name=thigh_twist_01_r
  bone2Name=pelvis
  swing1LimitDegrees=45
  swing1Motion=Free
  swing2LimitDegrees=45
  swing2Motion=Free
  twistLimitDegrees=45
  twistMotion=Free
 }
 {
  bone1Name=thigh_twist_01_l
  bone2Name=pelvis
  swing1LimitDegrees=45
  swing1Motion=Free
  swing2LimitDegrees=45
  swing2Motion=Free
  twistLimitDegrees=45
  twistMotion=Free
 }
 {
  bone1Name=dyn_jacket_l_02
  bone2Name=dyn_jacket_l_fr_02
  swing1LimitDegrees=20
  swing2LimitDegrees=20
  twistLimitDegrees=45
  twistMotion=Free
 }
 {
  bone1Name=dyn_jacket_r_bk_02
  bone2Name=dyn_jacket_r_02
  swing1LimitDegrees=20
  swing2LimitDegrees=20
  twistLimitDegrees=45
  twistMotion=Free
 }
 {
  bone1Name=dyn_jacket_r_fr_02
  bone2Name=dyn_jacket_r_02
  swing1LimitDegrees=10
  swing2LimitDegrees=10
  twistLimitDegrees=45
  twistMotion=Free
 }
 {
  bone1Name=dyn_jacket_l_bk_02
  bone2Name=dyn_jacket_l_02
  swing1LimitDegrees=20
  swing2LimitDegrees=20
  twistLimitDegrees=45
  twistMotion=Free
 }
 {
  bone1Name=dyn_jacket_bk_l_01
  bone2Name=dyn_jacket_l_bk_02
  swing1LimitDegrees=20
  swing2LimitDegrees=20
  twistLimitDegrees=45
  twistMotion=Free
 }
 {
  bone1Name=dyn_jacket_bk_r_01
  bone2Name=dyn_jacket_r_bk_02
  swing1LimitDegrees=20
  swing2LimitDegrees=20
  twistLimitDegrees=45
  twistMotion=Free
 }
]
