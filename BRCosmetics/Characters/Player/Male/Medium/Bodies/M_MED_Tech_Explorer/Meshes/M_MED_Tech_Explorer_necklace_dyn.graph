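# Physics asset for the dynamic necklace on the male medium Tech Explorer player body, defining capsule bodies and joints for its secondary motion.
# The necklace master bone locks its rotation to spine_05, which carries a large capsule for the upper torso.
# Three necklace capsules hang in a chain from the master, each joint with limited swing and a tighter twist.
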PhysicsAsset /BRCosmetics/Characters/Player/Male/Medium/Bodies/M_MED_Tech_Explorer/Meshes/M_MED_Tech_Explorer_necklace_dyn
bodies {
 dyn_necklace_01 [{
   shapeType=Capsule
   length=9.5258
   radius=1.1314
   shapeName=dyn_necklace_01_capsule
   center=<4.2277 0 0.0393>
   rotation=<rot -89.4673 0.001 -0.0009>
  }]
 dyn_necklace_02 [{
   shapeType=Capsule
   length=0.218
   radius=1.7637
   shapeName=dyn_necklace_02_capsule
   center=<0.7025 0 0.0573>
   rotation=<rot -85.3364 0.0007 -0.0006>
  }]
 dyn_necklace_03 [{
   shapeType=Capsule
   length=2.927
   radius=1.5564
   shapeName=dyn_necklace_03_capsule
   center=<2.1731 0 0.1773>
   rotation=<rot -85.3367 0.0008 -0.0007>
  }]
 dyn_necklace_MSTR [{shapeType=Capsule, length=0.5, radius=0.5}]
 spine_05 [{shapeType=Capsule, length=9.4509, radius=14.9611, center=<-6.0943 -2.8082 0>}]
}
constraints [
 {
  bone1Name=dyn_necklace_MSTR
  bone2Name=spine_05
  swing1LimitDegrees=45
  swing1Motion=Locked
  swing2LimitDegrees=45
  swing2Motion=Locked
  twistLimitDegrees=45
  twistMotion=Locked
 }
 {bone1Name=dyn_necklace_01, bone2Name=dyn_necklace_MSTR, swing1LimitDegrees=15, swing2LimitDegrees=30, twistLimitDegrees=5}
 {bone1Name=dyn_necklace_02, bone2Name=dyn_necklace_01, swing1LimitDegrees=30, swing2LimitDegrees=30, twistLimitDegrees=5}
 {bone1Name=dyn_necklace_03, bone2Name=dyn_necklace_02, swing1LimitDegrees=30, swing2LimitDegrees=30, twistLimitDegrees=5}
]
