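# Post-process style AnimBlueprint for the male medium Tech Explorer body that copies the attached parent mesh's pose and adds secondary motion to jacket, necklace, chains and belt buckle.
# Each update derives thigh-raise alphas from thigh rotation relative to the pelvis, driving additive offsets on jacket and chain bones.
# Bone constraints place chains and buckle before three base-bone-space RigidBody simulations, followed by small corrective offsets.
# Simulation gravity is reduced during skydiving until the parachute opens.
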AnimBlueprint /BRCosmetics/Characters/Player/Male/Medium/Bodies/M_MED_Tech_Explorer/Meshes/M_MED_Tech_Explorer_AnimBP
target_skeleton=&/M_MED_Tech_Explorer_Skeleton
graphs [
 AnimGraph {nodes [
   n1 = "Transform (Modify) Bone - Bone: dyn_jacket_l_fr_MSTR"(
    ComponentPose=Local To Component(LocalPose=Copy Pose From Mesh() {bUseAttachedParent=true})
    Alpha=Get thigh_up_l() -> (thigh_up_l:double)
    Translation="-2,0,0"
    Rotation="0,0,-70"
    Scale="1,1,1"
   ) {
    rotationMode=BMM_Additive
    rotationSpace=BCS_BoneSpace
    translationMode=BMM_Additive
    translationSpace=BCS_BoneSpace
    alphaBoolBlend {blendOption=Linear}
    boneToModify {boneName=dyn_jacket_l_fr_MSTR}
    rotation {roll=-70}
    translation {x=-2}
   }
   n2 = Get thigh_up_r() -> (thigh_up_r:double)
   n3 = "Transform (Modify) Bone - Bone: dyn_jacket_r_fr_MSTR"(ComponentPose=@n1, Alpha=@n2, Translation="-2.000,1,1", Rotation="0.000,0.000,70", Scale="1,1,1") {
    rotationMode=BMM_Additive
    rotationSpace=BCS_BoneSpace
    translationMode=BMM_Additive
    translationSpace=BCS_BoneSpace
    alphaBoolBlend {blendOption=Linear}
    boneToModify {boneName=dyn_jacket_r_fr_MSTR}
    rotation {roll=70}
    translation=<-2 1 1>
   }
   n4 = "Transform (Modify) Bone - Bone: dyn_jacket_r_MSTR"(ComponentPose=@n3, Alpha=@n2, Translation="-2.000,1,0", Rotation="0,0,0", Scale="1,1,1") {
    rotationMode=BMM_Additive
    rotationSpace=BCS_BoneSpace
    translationMode=BMM_Additive
    translationSpace=BCS_BoneSpace
    alphaBoolBlend {blendOption=Linear}
    boneToModify {boneName=dyn_jacket_r_MSTR}
    translation=<-2 1>
   }
   n5 = "Constraint - dyn_chains_ft_MSTR (PPP)"!(ComponentPose=@n4, Alpha=1, ConstraintWeights_0=0.676, ConstraintWeights_1:float=0.112, ConstraintWeights_2:float=0.212) -> (Pose) {
    alphaBoolBlend {blendOption=Linear}
    boneToModify {boneName=dyn_chains_ft_MSTR}
    constraintSetup [{targetBone {boneName=pelvis}}, {targetBone {boneName=spine_01}}, {targetBone {boneName=thigh_twist_01_r}}]
    constraintWeights [1,1,1]
   }
   n6 = "Constraint - dyn_chains_bk_MSTR (PP)"!(ComponentPose=@n5, Alpha=1, ConstraintWeights_0=0.3, ConstraintWeights_1:float=0.7) -> (Pose) {
    alphaBoolBlend {blendOption=Linear}
    boneToModify {boneName=dyn_chains_bk_MSTR}
    constraintSetup [{targetBone {boneName=pelvis}}, {targetBone {boneName=thigh_twist_01_r}}]
    constraintWeights [1,1]
   }
   n7 = "Constraint - dyn_belt_buckle_MSTR (PPP)"!(ComponentPose=@n6, Alpha=1, ConstraintWeights_0=0.54, ConstraintWeights_1:float=0.4, ConstraintWeights_2:float=0.06) -> (Pose) {
    alphaBoolBlend {blendOption=Linear}
    boneToModify {boneName=dyn_belt_buckle_MSTR}
    constraintSetup [{targetBone {boneName=pelvis}}, {targetBone {boneName=spine_01}}, {targetBone {boneName=spine_03}}]
    constraintWeights [1,1,1]
   }
   n8 = "Transform (Modify) Bone - Bone: dyn_chains_ft_MSTR"(ComponentPose=@n7, Alpha=@n2, Translation="-6,0,0", Rotation="0,0,0", Scale="1,1,1") {
    rotationMode=BMM_Additive
    rotationSpace=BCS_BoneSpace
    translationMode=BMM_Additive
    translationSpace=BCS_BoneSpace
    alphaBoolBlend {blendOption=Linear}
    boneToModify {boneName=dyn_chains_ft_MSTR}
    translation {x=-6}
   }
   n9 = Get gravity_override() -> (gravity_override:Vector)
   n10 = RigidBody!(ComponentPose=@n8, Alpha, OverrideWorldGravity:Vector=@n9, ExternalForce) -> (Pose) {
    bClampLinearTranslationLimitToRefPose=true
    bOverrideWorldGravity=true
    overridePhysicsAsset=&/M_MED_Tech_Explorer_jacket_dyn
    simulationSpace=BaseBoneSpace
    alphaBoolBlend {blendOption=Linear}
    baseBoneRef {boneName=pelvis}
    componentAppliedLinearAccClamp=<1000 1000 1000>
    componentLinearAccScale=<0.35 0.35 0.4>
    componentLinearVelScale=<0.2 0.2 0.2>
    overrideWorldGravity {z=-980}
   }
   n11 = RigidBody(ComponentPose=@n10, OverrideWorldGravity=@n9) {
    bClampLinearTranslationLimitToRefPose=true
    bOverrideWorldGravity=true
    overridePhysicsAsset=&/M_MED_Tech_Explorer_necklace_dyn
    simulationSpace=BaseBoneSpace
    alphaBoolBlend {blendOption=Linear}
    baseBoneRef {boneName=spine_05}
    componentAppliedLinearAccClamp=<1000 500 500>
    componentLinearAccScale=<0.3 0.2 0.2>
    componentLinearVelScale=<0.3 0.1 0.1>
    overrideWorldGravity {z=-980}
   }
   n12 = RigidBody(ComponentPose=@n11, OverrideWorldGravity=@n9) {
    bClampLinearTranslationLimitToRefPose=true
    bOverrideWorldGravity=true
    overridePhysicsAsset=&/M_MED_Tech_Explorer_chains_dyn
    simulationSpace=BaseBoneSpace
    alphaBoolBlend {blendOption=Linear}
    baseBoneRef {boneName=thigh_r}
    componentAppliedLinearAccClamp=<100 100 100>
    overrideWorldGravity {z=-980}
    simSpaceSettings {maxAngularAcceleration=10, maxAngularVelocity=5, maxLinearAcceleration=500, maxLinearVelocity=350, worldAlpha=1}
   }
   n13 = "Transform (Modify) Bone - Bone: dyn_chains_01_ft"!(ComponentPose=@n12, bAlphaBoolEnabled:bool=True, Translation="0,0,0", Rotation="-3,0,0") -> (Pose) {
    alphaInputType=Bool
    rotationMode=BMM_Additive
    translationMode=BMM_Additive
    alphaBoolBlend {blendInTime=0.1, blendOutTime=0.1}
    boneToModify {boneName=dyn_chains_01_ft}
    rotation {pitch=-3}
   }
   n14 = "Transform (Modify) Bone - Bone: dyn_chains_01_ft"(ComponentPose=@n13, bAlphaBoolEnabled=True, Translation="0,0,0", Rotation="-3,0,0") {
    alphaInputType=Bool
    rotationMode=BMM_Additive
    translationMode=BMM_Additive
    alphaBoolBlend {blendInTime=0.1, blendOutTime=0.1}
    boneToModify {boneName=dyn_chains_01_ft}
    rotation {pitch=-3}
   }
   n15 = "Transform (Modify) Bone - Bone: dyn_jacket_r_bk_01"!(ComponentPose=@n14, bAlphaBoolEnabled:bool=True, Translation="-0.000,-3.157,0.000", Rotation="-0.000,0.000,5.911") -> (Pose) {
    alphaInputType=Bool
    rotationMode=BMM_Additive
    translationMode=BMM_Additive
    alphaBoolBlend {blendInTime=0.1, blendOutTime=0.1}
    boneToModify {boneName=dyn_jacket_r_bk_01}
    rotation {roll=5.911}
    translation {y=-3.157}
   }
   Output Pose(Result=Component To Local(ComponentPose=@n15)) {layerGroup=None}
   Get bIsAccelerating2D()
  ]}
 EventGraph {nodes [
   Event BlueprintUpdateAnimation()
   n1 = BreakTransform(InTransform=GetDeltaTransformFromRefPose(self=GetOwnerSkeletalMeshComponent(), BoneName=thigh_l, BaseName=pelvis))
   > Set thigh_up_l(thigh_up_l:double=MapRangeClamped(Value=BreakRotator(InRot=@n1.Rotation).Yaw, InRangeA=-20, InRangeB=-100)) -> (Output_Get:double)
   n2 = BreakTransform(InTransform=GetDeltaTransformFromRefPose(self=GetOwnerSkeletalMeshComponent(), BoneName=thigh_r, BaseName=pelvis))
   > Set thigh_up_r(thigh_up_r:double=MapRangeClamped(Value=BreakRotator(InRot=@n2.Rotation).Yaw, InRangeA=-20, InRangeB=-100)) -> (Output_Get:double)
   > n3 = Branch(Condition=Get bIsSkyDiving())
   > n4 = Branch(Condition=Get bIsParachuteOpened())
   Set gravity_override(execute=[@n4.then, @n3.else], gravity_override="0, 0,-980")
   Set gravity_override(execute=@n4.else, gravity_override="0, 0,-300")
  ]}
]
variables [
 gravity_override {type=Vector}
 thigh_up_l {type=number}
 thigh_up_r {type=number}
]
