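# Control Rig for the male medium Tech Explorer character that shifts dyn_galss, apparently the glasses bone, in response to cheek movement.
# Each inner cheek bone's Z position relative to faceAttach is remapped and clamped into a weight.
# That weight drives an additive local nudge and tilt on dyn_galss, with opposite roll for left and right.
# The dynamic hair bones are in the hierarchy but untouched by the graph.
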ControlRigBlueprint /BRCosmetics/Characters/Player/Male/Medium/Bodies/M_MED_Tech_Explorer/Meshes/Parts/M_MED_Tech_Explorer_CtrlRig
graphs [
 RigVMFunctionLibrary {links [], nodes []}
 RigVMModel {nodes [
   n1 = &RigUnit_BeginExecution()
   /Script/ControlRig.RigUnit_ModifyBoneTransforms(
    BoneToModify=((Bone="dyn_galss",Transform=(Rotation=(X=0,Y=0,Z=0,W=1),Translation=(X=0.3,Y=0,Z=0),Scale3D=(X=1,Y=1,Z=1))))
    Mode=AdditiveLocal
    WeightMaximum=1
    WeightMinimum=0
    ExecutePin=@n1.ExecutePin
    Weight=&RigVMFunction_MathFloatRemap(
     SourceMaximum=5.2
     SourceMinimum=3.7
     TargetMinimum=0
     bClamp=true
     Value=&RigUnit_GetRelativeBoneTransform(Bone=L_cheek_inner, Space=faceAttach).Transform.Translation.Z
    ).Result
    BoneToModify.0.Transform.Rotation=/Script/RigVM.RigVMFunction_MathQuaternionFromRotator(Rotator=(Pitch=1,Yaw=0,Roll=1)) {is=struct RigVMUnitNode}.Result
   ) {is=struct RigVMUnitNode}
   > &RigUnit_ModifyBoneTransforms(
    BoneToModify=((Bone="dyn_galss",Transform=(Rotation=(X=0,Y=0,Z=0,W=1),Translation=(X=0.3,Y=0,Z=0),Scale3D=(X=1,Y=1,Z=1))))
    Mode=AdditiveLocal
    WeightMaximum=1
    WeightMinimum=0
    Weight=&RigVMFunction_MathFloatRemap(
     SourceMaximum=5.2
     SourceMinimum=3.7
     TargetMinimum=0
     bClamp=true
     Value=&RigUnit_GetRelativeBoneTransform(Bone=R_cheek_inner, Space=faceAttach).Transform.Translation.Z
    ).Result
    BoneToModify.0.Transform.Rotation=&RigVMFunction_MathQuaternionFromRotator(Rotator=(Pitch=1,Yaw=0,Roll=-1)).Result
   )
  ]}
]
hierarchy [
 root {loc=<0 0 0>, rot=<0 0 0 1>}
 pelvis {parent=root, loc=<0 -0.7071 93.1191>, rot=<0.0127 0.707 -0.0127 -0.707>}
 spine_01 {parent=pelvis, loc=<6.7494 -0.6388 0>, rot=<0 0 0.034 -0.9994>}
 spine_02 {parent=spine_01, loc=<7.2518 -0.035 0>, rot=<0 0 0.0238 -0.9997>}
 spine_03 {parent=spine_02, loc=<6.0104 0 0>, rot=<0 0 0.0149 -0.9999>}
 spine_04 {parent=spine_03, loc=<8.1024 0 0>, rot=<0 0 -0.0132 -0.9999>}
 spine_05 {parent=spine_04, loc=<10.0146 0 0>, rot=<0 0 -0.0015 -1>}
 neck_01 {parent=spine_05, loc=<11.3138 0.1529 0>, rot=<0 0 -0.2003 -0.9797>}
 neck_02 {parent=neck_01, loc=<4.5627 0 0>, rot=<0 0 0.0504 -0.9987>}
 head {parent=neck_02, loc=<5.3914 0 0>, rot=<0 0 0.115 -0.9934>}
 faceAttach {parent=head, loc=<-2.6072 0.7431 0.0021>, rot=<0.0082 -0.7071 -0.0082 -0.7071>, scale=<0.89 0.89 0.89>}
 L_cheek_inner {parent=faceAttach, loc=<4.1563 7.6692 3.5932>, rot=<-0.0484 -0.2135 0.3257 -0.9198>}
 C_nose_bridge {parent=faceAttach, loc=<0.0085 11.0337 5.611>, rot=<-0.2218 -0.0176 -0.0049 -0.9749>}
 R_cheek_inner {parent=faceAttach, loc=<-4.1407 7.6778 3.5837>, rot=<-0.0484 0.2136 -0.3256 -0.9198>}
 dyn_galss {parent=head, loc=<5.1224 0.1173 0>, rot=<0.4704 0.5279 0.5279 0.4704>}
 dyn_hair_A_01 {parent=head, loc=<11.0108 9.2363 1.2912>, rot=<0.2184 -0.6285 -0.5781 -0.4723>}
 dyn_hair_A_02 {parent=dyn_hair_A_01, loc=<3.8057 0.1071 0.0423>, rot=<0.0365 -0.0529 -0.4269 -0.902>}
 dyn_hair_A_03 {parent=dyn_hair_A_02, loc=<2.9612 -1.0095 -0.1849>, rot=<-0.0242 -0.1193 0.4595 -0.8798>}
 dyn_hair_A_04 {parent=dyn_hair_A_03, loc=<2.1134 -1.1781 -0.8958>, rot=<-0.1302 -0.459 -0.1123 -0.8716>}
 dyn_hair_B_01 {parent=head, loc=<12.5702 8.2343 -0.8456>, rot=<0.1381 -0.8452 -0.3048 -0.4168>}
 dyn_hair_B_02 {parent=dyn_hair_B_01, loc=<3.2978 0.1662 0.105>, rot=<0.0516 -0.1995 -0.1068 -0.9727>}
 dyn_hair_B_03 {parent=dyn_hair_B_02, loc=<3.7554 0.0722 -0.1453>, rot=<0.0706 -0.0204 0.3744 -0.9244>}
 dyn_hair_C_01 {parent=head, loc=<10.7456 7.1347 -5.2818>, rot=<0.4652 -0.8373 -0.2336 -0.167>}
 dyn_hair_C_02 {parent=dyn_hair_C_01, loc=<2.2628 -0.0587 0.0079>, rot=<0.0598 -0.0717 -0.3223 -0.942>}
 dyn_hair_C_03 {parent=dyn_hair_C_02, loc=<2.4158 0.173 0.0234>, rot=<-0.0012 0.117 -0.0605 -0.9913>}
 dyn_hair_D_01 {parent=head, loc=<10.9892 0.9088 -6.4216>, rot=<-0.0512 -0.9881 -0.1274 -0.0694>}
 dyn_hair_D_02 {parent=dyn_hair_D_01, loc=<3.0131 -0.018 0.0261>, rot=<0.0277 0.0355 0.4767 -0.8779>}
]
