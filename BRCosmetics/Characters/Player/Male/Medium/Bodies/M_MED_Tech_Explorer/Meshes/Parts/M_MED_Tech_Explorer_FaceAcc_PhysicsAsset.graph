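# Physics asset for the face accessory mesh of the male medium Tech Explorer player body, covering the head, glasses and hair.
# Four hair strands (A to D) are capsule chains of two to four bodies, each rooted at the head capsule and joined bone to bone with narrow swing and twist limits.
# The glasses box (bone dyn_galss) attaches directly to the head with free swing and twist, so its angular limit values do not constrain it.
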PhysicsAsset /BRCosmetics/Characters/Player/Male/Medium/Bodies/M_MED_Tech_Explorer/Meshes/Parts/M_MED_Tech_Explorer_FaceAcc_PhysicsAsset
bodies {
 dyn_galss [{shapeType=Box, extentX=1.1395, extentY=11.1862, extentZ=3.4186, center=<9.2442 0 0>}]
 dyn_hair_A_01 [{shapeType=Capsule, length=2.6324, radius=1.3676, center=<1.7007 0 0>, rotation=<rot 90 0 0>}]
 dyn_hair_A_02 [{shapeType=Capsule, length=2.7012, radius=1.2988, center=<1.5545 0 0>, rotation=<rot 90 0 0>}]
 dyn_hair_A_03 [{shapeType=Capsule, length=2.0958, radius=0.9042, center=<1 0 0>, rotation=<rot 69.9978 -24.9992 0>}]
 dyn_hair_A_04 [{shapeType=Capsule, length=0.995, radius=0.6766, center=<0.6478 0.1636 0.1698>, rotation=<rot 81.9806 -106.321 -122.0082>}]
 dyn_hair_B_01 [{shapeType=Capsule, length=2.3736, radius=1.2402, center=<1.4895 0 0>, rotation=<rot 90 0 0>}]
 dyn_hair_B_02 [{shapeType=Capsule, length=2.9438, radius=1.0562, center=<1.9471 0 0>, rotation=<rot 90 0 0>}]
 dyn_hair_B_03 [{shapeType=Capsule, length=1.464, radius=0.7377, center=<1.3166 -0.2657 0>, rotation=<rot 90 26.565 66.5663>}]
 dyn_hair_C_01 [{shapeType=Capsule, length=1.2653, radius=1.2347, center=<1 0 0>, rotation=<rot 90 0 0>}]
 dyn_hair_C_02 [{shapeType=Capsule, length=1.6363, radius=0.8637, center=<1 0 0>, rotation=<rot 90 0 0>}]
 dyn_hair_C_03 [{shapeType=Capsule, length=1.1755, radius=0.6819, center=<0.8762 -0.2145 0.2206>, rotation=<rot 90 0 30.0025>}]
 dyn_hair_D_01 [{shapeType=Capsule, length=1.654, radius=0.846, center=<1.5549 0 0>, rotation=<rot 90 0 0>}]
 dyn_hair_D_02 [{shapeType=Capsule, length=1.0547, radius=0.7559, center=<0.9325 -0.4694 0>, rotation=<rot 90 90 139.9987>}]
 head [{shapeType=Capsule, length=2.0976, radius=6.1554, center=<7.0373 1.7302 0.0001>, rotation=<rot 90 -7.125 78.7942>}]
}
constraints [
 {
  bone1Name=dyn_galss
  bone2Name=head
  swing1LimitDegrees=45
  swing1Motion=Free
  swing2LimitDegrees=45
  swing2Motion=Free
  twistLimitDegrees=45
  twistMotion=Free
 }
 {bone1Name=dyn_hair_A_01, bone2Name=head, swing1LimitDegrees=15, swing2LimitDegrees=5, twistLimitDegrees=2}
 {bone1Name=dyn_hair_A_02, bone2Name=dyn_hair_A_01, swing1LimitDegrees=15, swing2LimitDegrees=5, twistLimitDegrees=2}
 {bone1Name=dyn_hair_A_03, bone2Name=dyn_hair_A_02, swing1LimitDegrees=15, swing2LimitDegrees=5, twistLimitDegrees=2}
 {bone1Name=dyn_hair_A_04, bone2Name=dyn_hair_A_03, swing1LimitDegrees=15, swing2LimitDegrees=5, twistLimitDegrees=2}
 {bone1Name=dyn_hair_B_01, bone2Name=head, swing1LimitDegrees=10, swing2LimitDegrees=5, twistLimitDegrees=2}
 {bone1Name=dyn_hair_B_02, bone2Name=dyn_hair_B_01, swing1LimitDegrees=10, swing2LimitDegrees=5, twistLimitDegrees=2}
 {bone1Name=dyn_hair_B_03, bone2Name=dyn_hair_B_02, swing1LimitDegrees=10, swing2LimitDegrees=5, twistLimitDegrees=2}
 {bone1Name=dyn_hair_C_01, bone2Name=head, swing1LimitDegrees=15, swing2LimitDegrees=5, twistLimitDegrees=2}
 {bone1Name=dyn_hair_C_02, bone2Name=dyn_hair_C_01, swing1LimitDegrees=15, swing2LimitDegrees=5, twistLimitDegrees=2}
 {bone1Name=dyn_hair_D_01, bone2Name=head, swing1LimitDegrees=15, swing2LimitDegrees=5, twistLimitDegrees=2}
 {bone1Name=dyn_hair_D_02, bone2Name=dyn_hair_D_01, swing1LimitDegrees=15, swing2LimitDegrees=5, twistLimitDegrees=2}
 {bone1Name=dyn_hair_C_03, bone2Name=dyn_hair_C_02, swing1LimitDegrees=15, swing2LimitDegrees=5, twistLimitDegrees=2}
]
